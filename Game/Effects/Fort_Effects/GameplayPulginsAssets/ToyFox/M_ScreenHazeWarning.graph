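# Post-process material for the ToyFox screen haze warning, adding a pulsing tint (red by default) over the scene color.
# Two sets of U/V edge masks shape the haze: one scales the added color, the other, times a remapped sine, drives the blend.
# A panning hex pattern, radially distorted with a sine-driven breathing offset, modulates the blend and gives way to solid haze where the edge mask saturates.
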Material /Game/Effects/Fort_Effects/GameplayPulginsAssets/ToyFox/M_ScreenHazeWarning
graph {
 properties {bAutomaticallySetUsageInEditor=false, materialDomain=MD_PostProcess, usedShadingModels=Default Lit}
 nodes [
  n1 = Sine(None=Multiply(A=Time(), B=ScalarParameter() {defaultValue=1.5, parameterName=Scale}))
  Abs(None=@n1)
  Constant()
  n2 = SceneTexture() {sceneTextureId=PPI_PostProcessInput0}
  n3 = TextureCoordinate()
  n4 = MaterialFunctionCall(Float2=@n3) {materialFunction=&BreakOutFloat2Components}
  n5 = VectorParameter() {group=Edge, parameterName=HLEdge_EdgeMask_Control, channelNames {a="", b=Power, g=Multiply, r=Offset}, defaultValue=<rgba 0 15 1 8>}
  n6 = OneMinus(None=MaterialFunctionCall!(UorV=@n4.R, EdgeControl=@n5) {materialFunction=/Game/Effects/Niagara/MasterMaterials/Functions/Common/FXCommon_EdgeMask_Simple}.Mask)
  n7 = VectorParameter() {group=Edge, parameterName=HLEdgeV_EdgeMask_Control, channelNames {a="", b=Power, g=Multiply, r=Offset}, defaultValue=<rgba 0.5 38.9157 1 10>}
  n8 = OneMinus(None=MaterialFunctionCall(UorV=@n4.G, EdgeControl=@n7) {materialFunction=&FXCommon_EdgeMask_Simple}.Mask)
  n9 = Saturate(None=Add(A=Power(Base=@n6, Exp=@n5.A), B=Power(Base=@n8, Exp=@n7.A)))
  n10 = Add(A=Multiply(A=@n9, B=ScalarParameter() {defaultValue=10, parameterName=Sca}))
  n11 = VectorParameter() {parameterName=WarningColor, defaultValue=<rgba 1 0.1 0.1 1>}
  n12 = MaterialFunctionCall(Float2=@n3) {materialFunction=&BreakOutFloat2Components}
  n13 = VectorParameter() {group=Edge, parameterName=Edge_EdgeMask_Control, channelNames {a="", b=Power, g=Multiply, r=Offset}, defaultValue=<rgba 0.0832 2 1 3>}
  n14 = OneMinus(None=MaterialFunctionCall(UorV=@n12.R, EdgeControl=@n13) {materialFunction=&FXCommon_EdgeMask_Simple}.Mask)
  n15 = VectorParameter() {group=Edge, parameterName=EdgeV_EdgeMask_Control, channelNames {a="", b=Power, g=Multiply, r=Offset}, defaultValue=<rgba 0.0808 2 1 4>}
  n16 = OneMinus(None=MaterialFunctionCall(UorV=@n12.G, EdgeControl=@n15) {materialFunction=&FXCommon_EdgeMask_Simple}.Mask)
  n17 = Saturate(None=Add(A=Power(Base=@n14, Exp=@n13.A), B=Power(Base=@n16, Exp=@n15.A)))
  n18 = Constant() {r=1}
  n19 = MaterialFunctionCall(
   Input=@n1
   Input Low=Constant() {r=-1}
   Input High=@n18
   Target Low=ScalarParameter() {defaultValue=0.2, parameterName=Low}
   Target High=@n18
  ) {materialFunction=&RemapValueRange}
  n20 = TextureCoordinate()
  n21 = Constant2Vector() {g=0.5, r=0.5}
  n22 = ScalarParameter() {defaultValue=0.75, parameterName=HexUVRad}
  n23 = Power(
   Base=OneMinus(None=SphereMask(A=@n20, B=@n21, Radius=@n22) {attenuationRadius=1, hardnessPercent=0})
   Exp=ScalarParameter() {defaultValue=2.5, parameterName=HexUVPower}
  ) {constExponent=3}
  n24 = TextureCoordinate()
  n25 = Constant2Vector() {g=0.5, r=0.5}
  n26 = Sine(None=Multiply(A=Time(), B=ScalarParameter() {defaultValue=0.25, parameterName=HexBreathTime}))
  n27 = Constant() {r=-1}
  n28 = Constant() {r=1}
  n29 = VectorParameter() {parameterName=HexBreath, defaultValue=<rgba 0.8 1 0.7 1>}
  n30 = Multiply(
   A=Normalize(VectorInput=Subtract(A=@n24, B=@n25))
   B=LinearInterpolate(
    A=ScalarParameter() {parameterName=STE}
    B=ScalarParameter() {defaultValue=-0.75, parameterName=STE2}
    Alpha=MaterialFunctionCall(Input=@n26, Input Low=@n27, Input High=@n28, Target Low=@n29.R, Target High=@n29.R) {materialFunction=&RemapValueRange}.Result
   )
  )
  n31 = MaterialFunctionCall(Float2=ViewProperty() {property=MEVP_ViewSize}.Property) {materialFunction=&BreakOutFloat2Components}
  n32 = Divide(A=@n31.R, B=@n31.R)
  n33 = VectorParameter() {group=HexTexture, parameterName=HexTexture_UVControl, channelNames {a=VPan, b=UPan, g=VScale, r=UScale}, defaultValue=<rgba 2 2.5 0 0>}
  n34 = MaterialFunctionCall(
   UVControl_RGB=MaterialFunctionCall(X=Multiply(A=@n32, B=@n33.UScale), Y=@n33.VScale, Z=@n33.VScale) {materialFunction=&MakeFloat3}.Result
   UVControl_A=@n33.VPan
  ) {materialFunction=&FXCommon_TextureSample}
  n35 = VectorParameter() {parameterName=SpeedHex, defaultValue=<rgba 0 0.02 0 1>}
  n36 = Multiply(
   A=MaterialFunctionCall(X=@n35.R, Y=@n35.R) {materialFunction=&MakeFloat2}.Result
   B=Time() {bOverride_Period=true, period=6000}
  )
  n37 = TextureObjectParameter() {
   group=HexTexture
   parameterName=HexTexture_Texture
   samplerType=SAMPLERTYPE_Alpha
   sortPriority=0
   texture=&/T_UI_ToyFox_Pattern_02
  }
  n38 = TextureSample(UVs=Add(A=Multiply(A=@n23, B=@n30), B=Add(A=@n34.UV, B=@n36)), Tex=@n37)
  n39 = Multiply(
   A=Add(A=Desaturation(None=@n38.RGB), B=ScalarParameter() {group=HexTexture, parameterName=HexTexture_GrayAdjust_Add})
   B=ScalarParameter() {defaultValue=1, group=HexTexture, parameterName=HexTexture_GrayAdjust_Mul}
  )
  n40 = Saturate(None=Power(Base=Saturate(None=@n39), Exp=ScalarParameter() {defaultValue=1, group=HexTexture, parameterName=HexTexture_GrayAdjust_Power}))
  n41 = VectorParameter() {parameterName=HexRef_Refine, defaultValue=<rgba 5 10 2 0.5>}
  n42 = OneMinus(None=Power(Base=OneMinus(None=@n40), Exp=@n41.R))
  n43 = LinearInterpolate(A=Saturate(None=Multiply(A=@n42, B=@n41.G)), B=Power(Base=Multiply(A=@n40, B=@n41.B)) {constExponent=0.5}, Alpha=@n41.A)
  n44 = MaterialFunctionCall(
   Input=OneMinus(None=Saturate(None=@n43))
   Input Low=Constant()
   Input High=Constant() {r=1}
   Target Low=Multiply(B=ScalarParameter() {defaultValue=0.5908, parameterName=RWR}) {constA=0.65}
   Target High=MaterialFunctionCall(Input=@n26, Input Low=@n27, Input High=@n28, Target Low=@n29.B, Target High=@n29.B) {materialFunction=&RemapValueRange}.Result
  ) {materialFunction=&RemapValueRange}
  n45 = Saturate(None=Multiply(A=@n9, B=ScalarParameter() {defaultValue=2, parameterName=HexMaskSc}))
  LinearInterpolate(
   A=ComponentMask(None=@n2.Color) {b=true, g=true, r=true}
   B=Add(A=Multiply(A=@n10, B=@n11.RGB), B=ComponentMask(None=@n2.Color) {b=true, g=true, r=true})
   Alpha=Multiply(A=Multiply(A=@n17, B=@n19.Result), B=LinearInterpolate(A=@n44.Result, Alpha=@n45))
  )
 ]
}
